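# Corrective Control Rig for the QuarterClasp Pecan character's face accessories, adjusting a head knot and a neck mask to follow body poses.
# Knot bones are parent-constrained to head and neck, then offset by weights remapped from head rotation relative to neck and upper spine.
# The mask base blends neck_01 and spine_05, then rotates with neck bend and shifts by neck yaw.
# Clamped spine and neck translation deltas from initial pose add further mask offsets.
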ControlRigBlueprint /BRCosmetics/Characters/Player/Male/Medium/Bodies/M_MED_QuarterClasp_Pecan/Meshes/Parts/M_MED_QuarterClasp_Pecan_FaceAcc_CtrlRig
graphs [
 RigVMFunctionLibrary {links [], nodes []}
 RigVMFunction_Sequence_3_ContainedGraph {nodes [
   n1 = RigVMFunctionEntryNode()
   n2 = &RigVMFunction_Sequence(ExecuteContext=@n1.ExecuteContext)
   n3 = &RigVMFunction_Sequence(ExecuteContext=@n2.B)
   RigVMFunctionReturnNode(B=@n3.A, C=@n3.B, A=@n2.A)
  ]}
 RigVMModel {nodes [
   RigUnit_BeginExecution = &RigUnit_BeginExecution()
   RigVMFunction_Sequence = &RigVMFunction_Sequence(ExecuteContext=@RigUnit_BeginExecution.ExecutePin)
   &RigUnit_ParentConstraint(
    AdvancedSettings=(InterpolationType=Average,RotationOrderForFilter=XZY)
    Child=(Type=Bone,Name="deform_knot_base")
    Parents=((Item=(Type=Bone,Name="head"),Weight=1),(Item=(Type=Bone,Name="neck_01"),Weight=1))
    Weight=1
    ExecutePin=@RigVMFunction_Sequence.A
   )
   > n1 = &RigUnit_ModifyTransforms(
    ItemToModify=((Item=(Type=Bone,Name="head"),Transform=(Rotation=(X=0,Y=0,Z=0,W=1),Translation=(X=0,Y=0,Z=0),Scale3D=(X=1,Y=1,Z=1))))
    Weight=1
    WeightMinimum=0
   )
   n2 = &RigVMFunction_Sequence(ExecuteContext=@n1.ExecutePin)
   n3 = &RigUnit_GetRelativeTransformForItem(Child=(Type=Bone,Name="head"), Parent=(Type=Bone,Name="neck_01"), bChildInitial=False, bParentInitial=False)
   &RigUnit_ModifyTransforms(
    ItemToModify=((Item=(Type=Bone,Name="deform_knot_base_l"),Transform=(Rotation=(X=0,Y=-0.766,Z=0,W=0.6428),Translation=(X=-0.5,Y=0,Z=0.5),Scale3D=(X=1,Y=1,Z=1))))
    WeightMinimum=0
    Weight=&RigVMFunction_MathDoubleRemap(
     SourceMaximum=-65
     SourceMinimum=-20
     TargetMaximum=0.5
     TargetMinimum=0
     Value=&RigVMFunction_MathQuaternionToEuler(RotationOrder=ZYX, Value=@n3.RelativeTransform.Rotation).Result.Z
    ).Result
    ExecutePin=@n2.A
   )
   RigVMFunction_MathQuaternionToEuler = &RigVMFunction_MathQuaternionToEuler(
    RotationOrder=ZYX
    Value=&RigUnit_GetRelativeTransformForItem(Child=(Type=Bone,Name="head"), Parent=(Type=Bone,Name="spine_05"), bChildInitial=False, bParentInitial=False).RelativeTransform.Rotation
   )
   Remap = &RigVMFunction_MathDoubleRemap(SourceMaximum=-65, SourceMinimum=0, TargetMaximum=1, TargetMinimum=0, Value=@RigVMFunction_MathQuaternionToEuler.Result.Z)
   &RigUnit_ModifyTransforms(
    ItemToModify=((Item=(Type=Bone,Name="deform_knot_base_r"),Transform=(Rotation=(X=0,Y=-0.0872,Z=0,W=0.9962),Translation=(X=-2,Y=0,Z=2),Scale3D=(X=1,Y=1,Z=1))))
    WeightMinimum=0
    Weight=@Remap.Result
    ExecutePin=@n2.B
   )
   Maximum = &RigVMFunction_MathDoubleMax(
    A=&RigVMFunction_MathDoubleRemap(SourceMaximum=-75, SourceMinimum=0, TargetMaximum=1, TargetMinimum=0, Value=@RigVMFunction_MathQuaternionToEuler.Result.X).Result
    B=&RigVMFunction_MathDoubleRemap(SourceMaximum=75, SourceMinimum=0, TargetMaximum=1, TargetMinimum=0, Value=@RigVMFunction_MathQuaternionToEuler.Result.X).Result
   )
   > &RigUnit_ModifyTransforms(
    ItemToModify="((Item=(Type=Bone,Name=\"deform_knot_base_l\"),Transform=(Rotation=(X=0,Y=-0.342,Z=0,W=0.9397),Translation=(X=0.5,Y=0,Z=0),Scale3D=(X=1,Y=1,Z=1))),(Item=(Type=Bone,Name=\"deform_knot_base_r\"),Transform=(Rotation=(X=0,Y=-0.342,Z=0,W=0.9397),Translation=(X=-0.5 ...(truncated)"
    WeightMinimum=0
    Weight=@Maximum.Result
   )
   &RigUnit_ModifyTransforms(
    ItemToModify=((Item=(Type=Bone,Name="spine_05"),Transform=(Rotation=(X=0,Y=0,Z=0,W=1),Translation=(X=0,Y=0,Z=0),Scale3D=(X=1,Y=1,Z=1))),(Item=(Type=Bone,Name="clavicle_r"),Transform=(Rotation=(X=0,Y=0,Z=0,W=1),Translation=(X=0,Y=0,Z=0),Scale3D=(X=1,Y=1,Z=1))))
    Weight=1
    WeightMinimum=0
    ExecutePin=@RigVMFunction_Sequence.B
   )
   > RigVMFunction_Sequence_2 = RigVMAggregateNode()
   &RigUnit_ParentConstraint(
    AdvancedSettings=(InterpolationType=Average,RotationOrderForFilter=XZY)
    Child=(Type=Bone,Name="dyn_mask_base")
    Parents=((Item=(Type=Bone,Name="neck_01"),Weight=0.5),(Item=(Type=Bone,Name="neck_02"),Weight=0),(Item=(Type=Bone,Name="spine_05"),Weight=1))
    Weight=1
    ExecutePin=@RigVMFunction_Sequence_2.A
   )
   > &RigUnit_ModifyTransforms(
    ItemToModify=((Item=(Type=Bone,Name="neck_01"),Transform=(Rotation=(X=0,Y=0,Z=0,W=1),Translation=(X=0,Y=0,Z=0),Scale3D=(X=1,Y=1,Z=1))))
    Weight=1
    WeightMinimum=0
   )
   Absolute = &RigVMFunction_MathDoubleAbs(Value=&RigVMFunction_MathQuaternionToEuler(
     RotationOrder=ZYX
     Value=&RigUnit_GetRelativeTransformForItem(Child=(Type=Bone,Name="neck_02"), Parent=(Type=Bone,Name="spine_05"), bChildInitial=False, bParentInitial=False).RelativeTransform.Rotation
    ).Result.X)
   RigVMFunction_MathQuaternionFromEuler = &RigVMFunction_MathQuaternionFromEuler(
    Euler=(X=0,Y=0,Z=0)
    Euler.Z=&RigVMFunction_MathDoubleRemap(SourceMaximum=80, SourceMinimum=15, TargetMaximum=35, TargetMinimum=0, Value=@Absolute.Result).Result
   )
   > &RigUnit_ModifyTransforms(
    ItemToModify=((Item=(Type=Bone,Name="dyn_mask_base"),Transform=(Rotation=(X=0,Y=0,Z=0.766,W=0.6428),Translation=(X=0,Y=0,Z=0),Scale3D=(X=1,Y=1,Z=1))))
    Weight=1
    WeightMinimum=0
    ItemToModify.0.Transform.Rotation=@RigVMFunction_MathQuaternionFromEuler.Result
   )
   Get Delta Ref = RigVMFunctionReferenceNode(Item=(Type=Bone,Name="neck_01"), ExecuteContext=@RigVMFunction_Sequence_2.B)
   > &RigUnit_ModifyTransforms(
    ItemToModify=((Item=(Type=Bone,Name="dyn_mask_base"),Transform=(Rotation=(X=0,Y=0,Z=0,W=1),Translation=(X=-2,Y=2,Z=0),Scale3D=(X=1,Y=1,Z=1))))
    WeightMinimum=0
    Weight=&RigVMFunction_MathDoubleRemap(SourceMaximum=31, SourceMinimum=10, TargetMaximum=1, TargetMinimum=0, Value=@"Get Delta Ref".Result.Rotation.Yaw).Result
   )
   Multiply = /Script/RigVM.RigVMFunction_MathVectorMul(
    B=(X=-1,Y=-1,Z=-1)
    A=&RigUnit_GetRelativeTransformForItem(Child=(Type=Bone,Name="neck_01"), Parent=(Type=Bone,Name="spine_05"), bChildInitial=false, bParentInitial=false).RelativeTransform.Translation
   ) {is=struct RigVMUnitNode, template="Multiply::Execute(in A,in B,out Result)"}
   Add = /Script/RigVM.RigVMFunction_MathVectorAdd(
    B=&RigUnit_GetRelativeTransformForItem(Child=(Type=Bone,Name="neck_01"), Parent=(Type=Bone,Name="spine_05"), bChildInitial=true, bParentInitial=true).RelativeTransform.Translation
    A=@Multiply.Result
   ) {is=struct RigVMUnitNode, template="Add::Execute(in A,in B,out Result)"}
   n4 = &RigVMFunction_MathVectorMul(B=(X=-0.5,Y=0.4,Z=-0.3), A=@Add.Result) {template="Multiply::Execute(in A,in B,out Result)"}
   Clamp = &RigVMFunction_MathDoubleClamp(Maximum=4, Value=@n4.Result.Y)
   > &RigUnit_ModifyTransforms(
    ItemToModify=((Item=(Type=Bone,Name="dyn_mask_base"),Transform=(Rotation=(X=0,Y=0,Z=0,W=1),Translation=(X=0,Y=0,Z=0),Scale3D=(X=1,Y=1,Z=1))))
    Weight=1
    WeightMinimum=0
    ItemToModify.0.Transform.Translation.Y=@Clamp.Result
    ItemToModify.0.Transform.Translation.X=&RigVMFunction_MathDoubleClamp(Minimum=-4, Value=@n4.Result.X).Result
    ItemToModify.0.Transform.Translation.Z=@n4.Result.Z
   )
   n5 = &RigUnit_GetRelativeTransformForItem(Child=(Type=Bone,Name="spine_05"), Parent=(Type=Bone,Name="spine_04"), bChildInitial=false, bParentInitial=false)
   n6 = &RigVMFunction_MathVectorMul(
    B=(X=0,Y=2,Z=0)
    A=&RigVMFunction_MathVectorAdd(
     B=&RigUnit_GetRelativeTransformForItem(Child=(Type=Bone,Name="spine_05"), Parent=(Type=Bone,Name="spine_04"), bChildInitial=true, bParentInitial=true).RelativeTransform.Translation
     A=&RigVMFunction_MathVectorMul(B=(X=-1,Y=-1,Z=-1), A=@n5.RelativeTransform.Translation) {template="Multiply::Execute(in A,in B,out Result)"}.Result
    ) {template="Add::Execute(in A,in B,out Result)"}.Result
   ) {template="Multiply::Execute(in A,in B,out Result)"}
   &RigUnit_ModifyTransforms(
    ItemToModify=((Item=(Type=Bone,Name="dyn_mask_base"),Transform=(Rotation=(X=0,Y=0,Z=0,W=1),Translation=(X=0,Y=4,Z=0),Scale3D=(X=1,Y=1,Z=1))))
    Weight=1
    WeightMinimum=0
    ItemToModify.0.Transform.Translation.Y=&RigVMFunction_MathDoubleClamp(Maximum=4, Value=@n6.Result.Y).Result
    ExecutePin=@RigVMFunction_Sequence_2.C
   )
   &RigVMFunction_MathDoubleClamp(Minimum=-4, Value=@n6.Result.X)
   &RigVMFunction_MathQuaternionFromEuler(Euler=(X=0,Y=40,Z=0))
   &RigVMFunction_MathQuaternionFromEuler(Euler=(X=0,Y=20,Z=0))
  ]}
]
hierarchy [
 root {loc=<0 0 0>, rot=<0 0 0 1>}
 pelvis {parent=root, loc=<0 -0.7071 93.1191>, rot=<0.0127 0.707 -0.0127 -0.707>}
 spine_01 {parent=pelvis, loc=<6.7494 -0.6388 0>, rot=<0 0 0.034 -0.9994>}
 spine_02 {parent=spine_01, loc=<7.2518 -0.035 0>, rot=<0 0 0.0238 -0.9997>}
 spine_03 {parent=spine_02, loc=<6.0104 0 0>, rot=<0 0 0.0149 -0.9999>}
 spine_04 {parent=spine_03, loc=<8.1024 0 0>, rot=<0 0 -0.0132 -0.9999>}
 spine_05 {parent=spine_04, loc=<10.0146 0 0>, rot=<0 0 -0.0015 -1>}
 neck_01 {parent=spine_05, loc=<11.3138 0.1529 0>, rot=<0 0 -0.2003 -0.9797>}
 neck_02 {parent=neck_01, loc=<4.5627 0 0>, rot=<0 0 0.0504 -0.9987>}
 head {parent=neck_02, loc=<5.3914 0 0>, rot=<0 0 0.115 -0.9934>}
 FACIAL_C_FacialRoot {parent=head, loc=<2.1524 -0.746 -0.0002>, rot=<0.4879 -0.5118 -0.4879 -0.5118>}
 FACIAL_C_Forehead {parent=FACIAL_C_FacialRoot, loc=<-0.0182 -3.584 10.1254>, rot=<-0.0451 -0.0007 0.0026 -0.999>}
 FACIAL_L_ForeheadIn {parent=FACIAL_C_FacialRoot, loc=<1.4342 -3.5246 10.0504>, rot=<-0.0626 -0.0703 0.0067 -0.9955>}
 FACIAL_R_ForeheadIn {parent=FACIAL_C_FacialRoot, loc=<-1.4868 -3.5695 10.0515>, rot=<-0.0622 0.069 -0.0016 -0.9957>}
 FACIAL_L_ForeheadMid {parent=FACIAL_C_FacialRoot, loc=<2.792 -3.738 9.5791>, rot=<-0.0731 -0.1762 -0.0183 -0.9815>}
 FACIAL_R_ForeheadMid {parent=FACIAL_C_FacialRoot, loc=<-2.8758 -3.7421 9.5808>, rot=<-0.0722 0.175 0.0235 -0.9816>}
 FACIAL_L_ForeheadOut {parent=FACIAL_C_FacialRoot, loc=<4.022 -3.5954 8.815>, rot=<-0.1237 -0.2559 -0.077 -0.9556>}
 FACIAL_R_ForeheadOut {parent=FACIAL_C_FacialRoot, loc=<-4.1818 -3.5589 8.8234>, rot=<-0.1223 0.255 0.0821 -0.9557>}
 FACIAL_L_EyesackUpper {parent=FACIAL_C_FacialRoot, loc=<2.8148 -2.7298 10.0677>, rot=<0.0469 -0.0851 -0.0622 -0.9933>}
 FACIAL_R_EyesackUpper {parent=FACIAL_C_FacialRoot, loc=<-2.8867 -2.6637 10.0424>, rot=<0.0474 0.0832 0.0672 -0.9931>}
 FACIAL_L_EyelidUpperFurrow {parent=FACIAL_C_FacialRoot, loc=<2.7054 -2.574 9.8326>, rot=<-0.0322 -0.1514 -0.0774 -0.9849>}
 FACIAL_R_EyelidUpperFurrow {parent=FACIAL_C_FacialRoot, loc=<-2.7482 -2.5474 9.8244>, rot=<-0.0313 0.15 0.0824 -0.9848>}
 FACIAL_L_EyelidUpperA {parent=FACIAL_C_FacialRoot, loc=<2.7029 -2.0069 8.3924>, rot=<-0.1495 -0.0004 0.0026 -0.9888>}
 FACIAL_R_EyelidUpperA {parent=FACIAL_C_FacialRoot, loc=<-2.8052 -1.9559 8.3621>, rot=<-0.1495 -0.0004 0.0026 -0.9888>}
 FACIAL_L_EyelidLowerA {parent=FACIAL_C_FacialRoot, loc=<2.7029 -2.0069 8.3924>, rot=<0.2824 -0.0015 0.0022 -0.9593>}
 FACIAL_R_EyelidLowerA {parent=FACIAL_C_FacialRoot, loc=<-2.8052 -1.9558 8.3621>, rot=<0.2824 -0.0015 0.0022 -0.9593>}
 FACIAL_L_EyeCornerInner {parent=FACIAL_C_FacialRoot, loc=<1.7968 -1.6329 9.1362>, rot=<-0.0017 -0.0008 0.0026 -1>}
 FACIAL_R_EyeCornerInner {parent=FACIAL_C_FacialRoot, loc=<-1.8526 -1.5793 9.1299>, rot=<-0.0017 -0.0008 0.0026 -1>}
 FACIAL_L_EyeCornerOuter {parent=FACIAL_C_FacialRoot, loc=<3.9627 -1.6336 8.7119>, rot=<-0.0026 -0.3833 0.0017 -0.9236>}
 FACIAL_R_EyeCornerOuter {parent=FACIAL_C_FacialRoot, loc=<-4.0999 -1.5647 8.7068>, rot=<-0.0006 0.3818 0.003 -0.9242>}
 FACIAL_L_EyesackLower {parent=FACIAL_C_FacialRoot, loc=<2.9654 -0.476 8.5901>, rot=<-0.0243 -0.1404 -0.02 -0.9896>}
 FACIAL_R_EyesackLower {parent=FACIAL_C_FacialRoot, loc=<-3.0541 -0.4088 8.5674>, rot=<-0.0236 0.1389 0.025 -0.9897>}
 FACIAL_L_CheekInner {parent=FACIAL_C_FacialRoot, loc=<1.9861 0.2156 8.9162>, rot=<-0.1253 -0.0881 -0.1947 -0.9688>}
 FACIAL_R_CheekInner {parent=FACIAL_C_FacialRoot, loc=<-2.0057 0.2502 8.9129>, rot=<-0.1245 0.0872 0.1998 -0.968>}
 FACIAL_L_CheekOuter {parent=FACIAL_C_FacialRoot, loc=<3.693 0.5461 7.3784>, rot=<0.0527 -0.378 0.102 -0.9187>}
 FACIAL_R_CheekOuter {parent=FACIAL_C_FacialRoot, loc=<-3.8962 0.6423 7.3385>, rot=<0.0545 0.3762 -0.0974 -0.9198>}
 FACIAL_L_NasolabialBulge {parent=FACIAL_C_FacialRoot, loc=<3.4958 3.9814 7.3301>, rot=<0.0554 -0.3783 0.0255 -0.9237>}
 FACIAL_R_NasolabialBulge {parent=FACIAL_C_FacialRoot, loc=<-3.545 4.0591 7.3129>, rot=<0.0573 0.3765 -0.0209 -0.9244>}
 FACIAL_L_NasolabialFurrow {parent=FACIAL_C_FacialRoot, loc=<3.0847 2.6767 9.657>, rot=<-0.0355 -0.1259 -0.3615 -0.9232>}
 FACIAL_R_NasolabialFurrow {parent=FACIAL_C_FacialRoot, loc=<-3.099 2.7654 9.6473>, rot=<-0.0342 0.1245 0.3662 -0.9215>}
 FACIAL_L_CheekLower {parent=FACIAL_C_FacialRoot, loc=<3.9922 3.669 6.6832>, rot=<0.032 -0.4278 0.0593 -0.9014>}
 FACIAL_R_CheekLower {parent=FACIAL_C_FacialRoot, loc=<-4.057 3.7563 6.6652>, rot=<0.0341 0.4261 -0.0547 -0.9024>}
 FACIAL_C_Nose {parent=FACIAL_C_FacialRoot, loc=<0.0175 -0.2116 10.7773>, rot=<-0.132 -0.0005 0.0026 -0.9912>}
 FACIAL_C_NoseLower {parent=FACIAL_C_Nose, loc=<0.0385 2.52 -0.2469>, rot=<0.1305 0 0 -0.9914>}
 FACIAL_L_Nostril {parent=FACIAL_C_Nose, loc=<1.2964 1.9251 -0.4852>, rot=<0.0173 0.1305 0.0023 -0.9913>}
 FACIAL_R_Nostril {parent=FACIAL_C_Nose, loc=<-1.1899 1.8562 -0.5095>, rot=<0.0173 -0.1305 -0.0023 -0.9913>}
 FACIAL_C_NoseTip {parent=FACIAL_C_Nose, loc=<0.0509 1.8179 0.8345>, rot=<0.0872 0 0 -0.9962>}
 FACIAL_C_MouthUpper {parent=FACIAL_C_FacialRoot, loc=<-0.0094 4.2858 8.3121>, rot=<-0.0017 -0.0008 0.0026 -1>}
 FACIAL_C_LipUpper {parent=FACIAL_C_MouthUpper, loc=<0.0143 -0.4381 2.4568>, rot=<-0.0872 0 0 -0.9962>}
 FACIAL_L_LipUpper {parent=FACIAL_C_MouthUpper, loc=<1.1922 -0.2423 2.1693>, rot=<-0.0778 -0.1601 -0.0379 -0.9833>}
 FACIAL_R_LipUpper {parent=FACIAL_C_MouthUpper, loc=<-1.1683 -0.2529 2.1442>, rot=<-0.0778 0.1601 0.0379 -0.9833>}
 FACIAL_L_LipUpperOuter {parent=FACIAL_C_MouthUpper, loc=<1.8677 0.009 1.5516>, rot=<-0.0984 -0.2293 -0.0983 -0.9634>}
 FACIAL_R_LipUpperOuter {parent=FACIAL_C_MouthUpper, loc=<-1.8712 0.0195 1.5124>, rot=<-0.0984 0.2293 0.0983 -0.9634>}
 FACIAL_L_LipCorner {parent=FACIAL_C_MouthUpper, loc=<2.2358 0.3127 0.8902>, rot=<0 -0.2078 0 -0.9782>}
 FACIAL_R_LipCorner {parent=FACIAL_C_MouthUpper, loc=<-2.2499 0.3398 0.8526>, rot=<0 0.2078 0 -0.9782>}
 FACIAL_L_Masseter {parent=FACIAL_C_FacialRoot, loc=<5.8133 1.2457 4.9861>, rot=<0.068 0.1119 0.0105 -0.9913>}
 FACIAL_R_Masseter {parent=FACIAL_C_FacialRoot, loc=<-6.2475 1.4367 4.9807>, rot=<0.0674 -0.1139 -0.0056 -0.9912>}
 FACIAL_C_UnderChin {parent=FACIAL_C_FacialRoot, loc=<-0.0023 8.9757 5.8313>, rot=<0.4984 -0.002 0.0018 -0.867>}
 FACIAL_L_UnderChin {parent=FACIAL_C_FacialRoot, loc=<3.1962 7.1824 4.0927>, rot=<0.4003 -0.3567 0.0835 -0.84>}
 FACIAL_R_UnderChin {parent=FACIAL_C_FacialRoot, loc=<-3.347 7.3039 4.077>, rot=<0.4019 0.3533 -0.0799 -0.841>}
 FACIAL_C_LowerLipRotation {parent=FACIAL_C_FacialRoot, loc=<0.0002 1.4726 3.1009>, rot=<0.3404 -0.0016 0.0021 -0.9403>}
 FACIAL_C_MouthLower {parent=FACIAL_C_LowerLipRotation, loc=<-0.0325 -1.1754 5.8042>, rot=<-0.342 0 0 -0.9397>}
 FACIAL_C_LipLower {parent=FACIAL_C_MouthLower, loc=<0.0028 1.313 2.0199>, rot=<0.0872 0 0 -0.9962>}
 FACIAL_L_LipLower {parent=FACIAL_C_MouthLower, loc=<0.954 1.2553 1.8443>, rot=<0.0927 -0.1294 0.0272 -0.9869>}
 FACIAL_R_LipLower {parent=FACIAL_C_MouthLower, loc=<-0.9537 1.2629 1.8196>, rot=<0.0927 0.1294 -0.0272 -0.9869>}
 FACIAL_L_LipLowerOuter {parent=FACIAL_C_MouthLower, loc=<1.7245 0.8272 1.3651>, rot=<0.0913 -0.2321 0.0318 -0.9679>}
 FACIAL_R_LipLowerOuter {parent=FACIAL_C_MouthLower, loc=<-1.7301 0.8513 1.3212>, rot=<0.0913 0.2321 -0.0318 -0.9679>}
 FACIAL_C_Jaw {parent=FACIAL_C_FacialRoot, loc=<0.0002 1.4726 3.1009>, rot=<0.3404 -0.0016 0.0021 -0.9403>}
 FACIAL_C_Jawline {parent=FACIAL_C_Jaw, loc=<-0.0478 2.945 8.3916>, rot=<0.2586 0 0 -0.966>}
 FACIAL_L_ChinSide {parent=FACIAL_C_Jaw, loc=<2.6534 0.7604 7.6793>, rot=<-0.2208 -0.3738 0.0499 -0.8995>}
 FACIAL_R_ChinSide {parent=FACIAL_C_Jaw, loc=<-2.7419 0.8273 7.7061>, rot=<-0.2208 0.3738 -0.0499 -0.8995>}
 FACIAL_C_Chin {parent=FACIAL_C_Jaw, loc=<-0.0477 1.3979 8.3215>, rot=<-0.2591 0 0 -0.9659>}
 FACIAL_L_Ear {parent=FACIAL_C_FacialRoot, loc=<6.215 -0.117 1.4991>, rot=<-0.028 -0.0014 -0.0236 -0.9993>}
 FACIAL_R_Ear {parent=FACIAL_C_FacialRoot, loc=<-6.806 0.012 1.4405>, rot=<-0.028 -0.0001 0.0287 -0.9992>}
 FACIAL_L_Jawline {parent=FACIAL_C_FacialRoot, loc=<4.2877 5.1929 5.4269>, rot=<0.208 0.1565 -0.0521 -0.9641>}
 FACIAL_R_Jawline {parent=FACIAL_C_FacialRoot, loc=<-4.5194 5.341 5.4295>, rot=<0.2073 -0.1592 0.0567 -0.9636>}
 FACIAL_C_NoseBridge {parent=FACIAL_C_FacialRoot, loc=<-0.0007 -2.197 10.8833>, rot=<0.0421 -0.0009 0.0025 -0.9991>}
 FACIAL_L_EyelidUpperB {parent=FACIAL_C_FacialRoot, loc=<2.7029 -2.0069 8.3924>, rot=<-0.2266 -0.0002 0.0027 -0.974>}
 FACIAL_L_EyelidLowerB {parent=FACIAL_C_FacialRoot, loc=<2.7029 -2.0069 8.3924>, rot=<0.3157 -0.0016 0.0022 -0.9489>}
 FACIAL_R_EyelidLowerB {parent=FACIAL_C_FacialRoot, loc=<-2.8052 -1.9559 8.3621>, rot=<0.3239 -0.0016 0.0021 -0.9461>}
 FACIAL_R_EyelidUpperB {parent=FACIAL_C_FacialRoot, loc=<-2.8052 -1.9559 8.3621>, rot=<-0.2266 -0.0002 0.0027 -0.974>}
 deform_knot_base {parent=head, loc=<-1.867 -5.9096 0.1953>, rot=<0.1385 -0.706 0.6748 -0.1643>}
 deform_knot_base_l {parent=deform_knot_base, loc=<0.1699 1.1158 0.4131>, rot=<0.1094 -0.0462 -0.1528 -0.9811>}
 deform_knot_l {parent=deform_knot_base_l, loc=<0 0 0>, rot=<0 0 0 1>}
 deform_knot_base_r {parent=deform_knot_base, loc=<0.3145 -1.263 0.1066>, rot=<0 0 0.244 -0.9698>}
 deform_knot_r {parent=deform_knot_base_r, loc=<0 0 0>, rot=<0 0 0 1>}
 dyn_mask_base {parent=neck_01, loc=<3.1929 8.4944 0>, rot=<-0.2717 -0.9624 0 0>}
 dyn_mask_01 {parent=dyn_mask_base, loc=<0 0 0>, rot=<0 0 0 1>}
 clavicle_l {parent=spine_05, loc=<4.3575 -0.1466 -2.3223>, rot=<0.0465 -0.7289 -0.009 -0.6829>}
 clavicle_r {parent=spine_05, loc=<4.3575 -0.1466 2.3223>, rot=<0.7289 0.0465 -0.6829 0.009>}
]
